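; Athena infantry rifle weapon Blueprint that handles the cosmetic clip eject and its 'ching' sound.
; Fire FX calls ResetClipEject, then EjectClip after a short delay if the magazine is empty; DoOnce gates guard the particle and attached sound.
; Particle collisions spawn a ground impact sound on cosmetic-capable instances.
; Reload ejection binds to OnReloadStarted instead of reload FX, avoiding false triggers after sprinting (FORT-987613).
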(event EventBeginPlay
  ;[1 1 1] Spawn sound for clip hitting ground
  (|CallParentFunction)
  (Utilities|FlowControl|Sequence
    (:then_0
      (Networking|CanExecuteCosmeticEvents
        (:True
          (EventDispatchers|BindEventtoOnParticleCollide (Variables|Default|GetClipEject) (delegate Custom|Collide)))))
    (:then_1
      ;[0.9219 0.9219 0.9219 0.5] FORT-987613 - new logic to allow reload clip & 'ching' sound without iit triggering after sprint
      (EventDispatchers|AssignOnReloadStarted (delegate Custom|OnReloadStarted_Event)))))

(event Custom|Collide
  (params (EventName :type Name) (EmitterTime :type Float%20%28single-precision%29) (ParticleTime :type Integer) (Location :type Vector) (Velocity :type Vector) (Direction :type Vector) (Normal :type Vector) (BoneName :type Name) (PhysMat :type Physical%20Material%20Object%20Reference))
  ;[1 1 1] Spawn sound for clip hitting ground
  (Audio|SpawnSoundatLocation :Sound "/Game/Sounds/Fort_Weapons/Infantry_Rifle/MSS_InfantryRifle_Reload_Cartridge_Impact_Preset.MSS_InfantryRifle_Reload_Cartridge_Impact_Preset" :Location Location :VolumeMultiplier 1.0 :PitchMultiplier 1.0 :StartTime 0.0 :AttenuationSettings "/Game/Sounds/Attenuation/Fort_Weapon_Reload_Att.Fort_Weapon_Reload_Att" :bAutoDestroy true))

(event Custom|OnReloadStarted_Event
  (params (ReloadTime :type Float%20%28single-precision%29) (ReloadType :type EFortWeaponReloadType%20Enum :default "ReloadWholeClip"))
  ;[0.9219 0.9219 0.9219 0.5] FORT-987613 - new logic to allow reload clip & 'ching' sound without iit triggering after sprint
  (CallFunction|EjectClip))

(event EventOnPlayWeaponFireFX
  (params (bPersistentFire :type Boolean) (bSecondaryFire :type Boolean))
  (|CallParentFunction bPersistentFire bSecondaryFire)
  (CallFunction|ResetClipEject)
  (Utilities|FlowControl|Delay 0.1)
  (if (== (FortWeapon|GetLocalMagazineAmmoCount) 0)
    (CallFunction|EjectClip)))

(event EventOnPlayReloadFX
  (params (ReloadStage :type EFortReloadFXState%20Enum :default "ReloadStart"))
  (|CallParentFunction ReloadStage))

(event FortWeapon|EventOnLocalUnEquip
  ;[0.9688 0.3857 0 0.5] Event OnPlayReloadFX was triggering after ending a sprint, which is very weird
  ; FORT-987613 - disconnected this old logic as it was causing the clip & ching sound after shooting & sprinting, then stop sprinting
  ; Moved the Eject Clip call to onReloadStarted bind above
  (|CallParentFunction)
  (Variables|Default|SetIs%20Equiping false))

(event FortWeapon|EventOnEquippedAndAttached
  ;[0.9688 0.3857 0 0.5] Event OnPlayReloadFX was triggering after ending a sprint, which is very weird
  ; FORT-987613 - disconnected this old logic as it was causing the clip & ching sound after shooting & sprinting, then stop sprinting
  ; Moved the Eject Clip call to onReloadStarted bind above
  (|CallParentFunction)
  (Variables|Default|SetIs%20Equiping true))

(event Custom|EjectClip
  (Utilities|FlowControl|DoOnce true
    (:Completed
      (Components|Activation|Activate (Variables|Default|GetClipEject) false)
      (Audio|SpawnSoundAttached "/Game/Sounds/Fort_Weapons/Infantry_Rifle/MSS_InfantryRifle_Reload_Cartridge_Eject_Preset.MSS_InfantryRifle_Reload_Cartridge_Eject_Preset" (Variables|Visual|GetWeaponMesh))
      (Development|PrintString "Ejected"))))

(event Custom|ResetClipEject
  (Utilities|FlowControl|DoOnce true
    (:Completed
      (Components|Activation|Activate (Variables|Default|GetClipEject) false)
      (Audio|SpawnSoundAttached "/Game/Sounds/Fort_Weapons/Infantry_Rifle/MSS_InfantryRifle_Reload_Cartridge_Eject_Preset.MSS_InfantryRifle_Reload_Cartridge_Eject_Preset" (Variables|Visual|GetWeaponMesh))
      (Development|PrintString "Ejected"))))

(fn ConstructionScript
  (|CallParentFunction))
