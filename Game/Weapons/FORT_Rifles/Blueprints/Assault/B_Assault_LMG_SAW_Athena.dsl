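; B_Assault_LMG_SAW_Athena is the weapon Blueprint for the SAW light machine gun in the Assault rifle family, an Athena variant.
; Its construction script, BeginPlay and targeting-change event only forward to the parent class, and Tick is empty.
; The graphs add no logic of their own, so behavior comes from the parent class.
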(event EventBeginPlay
  (|CallParentFunction))

(event EventTick
  (params (DeltaSeconds :type Float%20%28single-precision%29)))

(event EventOnSetTargeting
  (params (bNewIsTargeting :type Boolean))
  (|CallParentFunction bNewIsTargeting))

(fn ConstructionScript
  (|CallParentFunction))
