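; Thermal-scoped surgical assault rifle weapon Blueprint that drives the scope's thermal vision post process while aiming.
; For a local viewer, targeting changes update whether custom depth rendering ignores significance.
; Scope PP caches the material instance and its ThermalScale layer parameter, zeroes it, then after a short delay ramps it by timeline over blend-in time.
; This manual toggle hides the player from their own scope; forced scope exit stops the timeline.
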(event EventOnSetTargeting
  (params (bNewIsTargeting :type Boolean))
  (Networking|CanExecuteCosmeticEvents
    (:True
      (bind _asfort_pawn (Utilities|Casting|CastToFortPawn (Game|GetInstigator)))
      (Utilities|IsValid (Pawn|GetLocalViewingPlayerController _asfort_pawn)
        (:Is%20Valid
          (|CallParentFunction bNewIsTargeting)
          (bind _asfort_pawn_2 (Utilities|Casting|CastToFortPawn (Game|GetInstigator)))
          (bind _asfort_player_controller (Utilities|Casting|CastToFortPlayerController (Pawn|GetLocalViewingPlayerController _asfort_pawn_2)))
          (Rendering|SetIgnoreSignifanceBasedCustomDepthRendering _asfort_player_controller (Camera|GetIsTargeting) 0))))))

(event EventPlayScopePP
  (bind (_weight _object) (Utilities|Struct|BreakWeightedBlendable (Utilities|Array|Get(aref) (Scope|GetScopePostProcessMaterial) 1)))
  ;[1 1 1] Manually Enabling/Disabling PP so player does not see himself through Thermal Scope
  (Utilities|FlowControl|Sequence
    (:then_0
      ;;[1 1 1] Assign thermal MID to variable
      (bind _asmaterial_instance_dynamic (Utilities|Casting|CastToMaterialInstanceDynamic _object))
      (bind _output_get (Variables|Default|SetThermalMID _asmaterial_instance_dynamic))
      (Variables|Default|SetThermalScale (Rendering|Material|GetParameterInfo _output_get "LayerParameter" "ThermalScale" "/Game/Weapons/FORT_Rifles/ThermalVision/ThermalAnimation/ML_AssaultSurgicalThermal.ML_AssaultSurgicalThermal")))
    (:then_1
      (Rendering|Material|SetScalarParameterValuebyInfo (Variables|Default|GetThermalMID) (Variables|Default|GetThermalScale) 0.0)
      (Utilities|FlowControl|RetriggerableDelay 0.1)
      (if (Camera|GetIsTargeting)
        (CallFunction|SetPostProcessParams 1.0)
        (Components|Timeline|SetPlayRate (Variables|B_Assault_Surgical_Thermal_Athena|GetScalePlayerThermal) (/ 1.0 (Variables|Default|GetBlendInTime)))
        (Utilities|FlowControl|DoOnce false
          (:Completed
            (bind (_direction _newtrack_0) (Timeline|ScalePlayerThermal PlayFromStart)
              (:Update
                (Rendering|Material|SetScalarParameterValuebyInfo (Variables|Default|GetThermalMID) (Variables|Default|GetThermalScale) _newtrack_0)))))))))

(event EventReverseScopePP
  ;[1 1 1] Manually Enabling/Disabling PP so player does not see himself through Thermal Scope
  ;;[0.1918 1 0.1475] No longer being called by parent
  (|CallParentFunction))

(event EventForceScopeBackImmediatly
  ;[1 1 1] Manually Enabling/Disabling PP so player does not see himself through Thermal Scope
  (|CallParentFunction)
  (Utilities|FlowControl|Sequence
    (:then_0
      (Utilities|FlowControl|DoOnce false
        (:Completed
          (Timeline|ScalePlayerThermal PlayFromStart))))
    (:then_1
      (Timeline|ScalePlayerThermal Stop))))

(fn ConstructionScript
  (|CallParentFunction))
