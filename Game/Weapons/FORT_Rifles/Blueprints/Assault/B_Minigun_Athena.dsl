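; Athena minigun weapon Blueprint that drives spin audio, overheat visuals and ammo-based audio.
; Charge events toggle Niagara and non-authority spin sounds scaled by a timeline; cooling and empty firing also spin down.
; Firing starts a timer mapping heat ratio to barrel-glow material opacity and an OnRep scalar, with a retriggerable delay preventing early timer shutdown.
; Overheating activates Niagara and prints debug text; ammo changes feed the pawn's audio parameter.
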(event EventOnAmmoCountChanged
  (|CallParentFunction)
  (Variables|Default|SetCurrentAmmo (FortWeapon|GetLocalMagazineAmmoCount)))

(event Custom|Start%20Spin%20Up%20Audio
  (bind _spinupsound (Variables|Default|GetSpinUpSound))
  ;[1 1 1] Handle volume scaling on the spin up and spin down audio cues
  (Networking|SwitchHasAuthority
    (:Remote
      (Variables|Default|SetCurrentSpinAudioComponent _spinupsound)
      (if (not (Audio|Components|Audio|IsPlaying _spinupsound))
        (Audio|Components|Audio|FadeIn _spinupsound 0.1 1.0 0.0 "Linear" :!class /Script/Engine.AudioComponent)
        (Audio|Components|Audio|Stop (Variables|Default|GetSpinDownSound))
        (else
          _))
      (bind (_direction _volume) (Timeline|SpinAudioVolumeTimeline Play)
        (:Update
          (Audio|Components|Audio|SetVolumeMultiplier (Variables|Default|GetCurrentSpinAudioComponent) _volume :!class /Script/Engine.AudioComponent))))))

(event Overheating|EventOnStartOverheated
  ;[1 1 1] Add overheat FX triggers here!
  (|CallParentFunction)
  (Development|PrintString "OVERHEATING")
  (Components|Activation|Activate (Variables|Default|GetOverheatedFX_Niagara) false))

(event Custom|Start%20Spin%20Down%20Audio
  (bind _spindownsound (Variables|Default|GetSpinDownSound))
  ;[1 1 1] Handle volume scaling on the spin up and spin down audio cues
  (Networking|SwitchHasAuthority
    (:Remote
      (Variables|Default|SetCurrentSpinAudioComponent _spindownsound)
      (if (not (Audio|Components|Audio|IsPlaying _spindownsound))
        (Audio|Components|Audio|FadeIn _spindownsound 0.1 1.0 0.0 "Linear" :!class /Script/Engine.AudioComponent)
        (Audio|Components|Audio|Stop (Variables|Default|GetSpinUpSound))
        (else
          _))
      (Timeline|SpinAudioVolumeTimeline Reverse))))

(event EventOnChargeUp
  ;[1 1 1] OnStartCharge - Sound
  (|CallParentFunction)
  (Components|Activation|Activate (Variables|Default|GetMinigunChargeUp_Niagara) false)
  (CallFunction|StartSpinUpAudio))

(event EventOnChargeDown
  ;[1 1 1] OnEndCharge - Sound
  (|CallParentFunction)
  (Utilities|FlowControl|Sequence
    (:then_0
      (CallFunction|StartSpinDownAudio))
    (:then_1
      (Components|Activation|Deactivate (Variables|Default|GetMinigunChargeUp_Niagara)))))

(event Overheating|EventOnStartCooling
  (|CallParentFunction)
  (CallFunction|StartSpinDownAudio))

(event EventOnOutOfAmmoFire
  (|CallParentFunction)
  (CallFunction|StartSpinDownAudio))

(event EventOnPlayWeaponFireFX
  (params (bPersistentFire :type Boolean) (bSecondaryFire :type Boolean))
  ;[1 1 1] OnPlayWeaponFireFX
  (|CallParentFunction bPersistentFire false)
  (bind _returnvalue (Utilities|Time|SetTimerbyEvent (EventDispatchers|CreateEvent :!event Custom|BarrelOverheatFXandSound) 0.0333 true))
  (Variables|Default|SetGlowTimer _returnvalue)
  ;[1 1 1] First Half second do not allow timer to end. This prevents it from turning off glow timer before it had time to spin up
  (Variables|Default|SetHitHeatThreshold false)
  (Utilities|FlowControl|RetriggerableDelay 0.5)
  (Variables|Default|SetHitHeatThreshold true))

(event Custom|BarrelOverheatFXandSound
  (bind _returnvalue (CallFunction|CurrentScalarValue))
  (bind _returnvalue_2 (Math|Float|Lerp 0.0 1.0 _returnvalue))
  ;[1 1 1] BarrelGlow - Handles ramping overheat FX and sound.  Timer shuts off when there is no heat in the minigun
  (if (> (CallFunction|CurrentScalarValue) 0.0)
    (Networking|CanExecuteCosmeticEvents
      (:True
        (CallFunction|UpdateOverheatScalar _returnvalue_2)
        (Rendering|Material|SetScalarParameterValue (Variables|Default|GetMuzzleGlowMID) "Opacity" _returnvalue_2 :!class /Script/Engine.MaterialInstanceDynamic)))
    (elif (Variables|Default|GetHitHeatThreshold)
      (Utilities|Time|ClearandInvalidateTimerbyHandle (Variables|Default|GetGlowTimer)))))

(event FortWeapon|EventOnEquippedAndAttached
  ;[1 1 1] OnEquip - Reset MuzzleGlowMID Opacity
  (|CallParentFunction)
  (Networking|CanExecuteCosmeticEvents
    (:True
      (CallFunction|UpdateOverheatScalar 0.0)
      (Rendering|Material|SetScalarParameterValue (Variables|Default|GetMuzzleGlowMID) "Opacity" 0.0 :!class /Script/Engine.MaterialInstanceDynamic)
      ;[1 1 1] Low Ammo Sound
      (bind _asfort_player_pawn (Utilities|Casting|CastToFortPlayerPawn (Game|GetInstigator)))
      (Variables|Default|SetAs%20Fort%20Player%20Pawn _asfort_player_pawn)
      (EventDispatchers|AssignOnLocalAmmoChanged (delegate Custom|OnLocalAmmoChanged_Event)))))

(event Custom|OnLocalAmmoChanged_Event
  (params (LocalCount :type Integer) (LocalRemaining :type Integer))
  ;[1 1 1] Low Ammo Sound
  (bind _as_fort_player_pawn (Variables|Default|GetAs%20Fort%20Player%20Pawn)
    (:then
      (Utilities|IsValid (Actor|GetComponentbyClass _as_fort_player_pawn "/Script/AudioGameplay.AudioParameterComponent")
        (:Is%20Valid
          (Audio|Parameter|SetIntegerParameter (Actor|GetComponentbyClass _as_fort_player_pawn "/Script/AudioGameplay.AudioParameterComponent") "Weapon.AmmoRemaining" LocalRemaining :!class /Script/AudioExtensions.AudioParameterControllerInterface))))
    (:else)))

(event FortWeapon|EventOnLocalUnEquip
  ;[1 1 1] Clear Timer Reset MuzzleGlowMID Opacity/Sound.  Keeps a phantom sound from playing.
  (|CallParentFunction)
  (Utilities|Time|ClearandInvalidateTimerbyHandle (Variables|Default|GetGlowTimer))
  (Networking|CanExecuteCosmeticEvents
    (:True
      (Rendering|Material|SetScalarParameterValue (Variables|Default|GetMuzzleGlowMID) "Opacity" 0.0 :!class /Script/Engine.MaterialInstanceDynamic))))

(event Custom|Update_Overheat_Scalar
  (params (NewScalar :type Float%20%28double-precision%29))
  ;[1 1 1 0.5] Update Overheat Scalar for Replication
  (Variables|Default|SetOverheat_Scalar NewScalar))

(event EventOnInitCosmeticAlterations
  (params (CosmeticMod :type Fort%20Cosmetic%20Modification%20Structure)))

(fn CurrentScalarValue
  (return (/ (Overheat|GetCurrentOverheatValue) (Overheat|GetOverheatingMaxValue))))

(fn OnRep_Overheat_Scalar
  (Rendering|Material|SetScalarParameterValue (Variables|Default|GetMuzzleGlowMID) "Opacity" (Variables|Default|GetOverheat_Scalar) :!class /Script/Engine.MaterialInstanceDynamic))

(fn ConstructionScript
  (Networking|CanExecuteCosmeticEvents
    (:True
      (bind _returnvalue (Rendering|Material|CreateDynamicMaterialInstance (Variables|Default|GetBarrelGlowFX) 0 0 "None" :!class /Script/Engine.PrimitiveComponent))
      (Variables|Default|SetMuzzleGlowMID _returnvalue))
    (:False
      _))
  (|CallParentFunction))
